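# Parent material for the Scamper gadget emote prop, adding masked scrolling textures to the shared 2023 character material functions.
# A Panner driven by ScrollSpeed scrolls Diffuse and SRM samples, which override base color, specular, roughness and metallic on the PreFX attributes.
# A selectable channel of the static FX mask, green by default, blends scrolled over original attributes.
# Specular, roughness and metallic all read SRM red, despite its per-channel description.
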Material /BRCosmetics/Animation/Props/Gadgets_Scamper/Materials/M_Emote_Prop_Scamper_PARENT
graph {
 properties {
  bHasPixelAnimation=true
  bUseMaterialAttributes=true
  bUsedWithMeshDeformer=true
  bUsedWithSkeletalMesh=true
  pixelDepthOffsetMode=PDOM_Legacy
  shadingModel=MSM_FromMaterialExpression
  usedShadingModels=Default Lit | Subsurface | Clear Coat | Subsurface Profile | Cloth
 }
 nodes [
  n1 = MaterialFunctionCall() {materialFunction=&MF_CharacterParent_2023_PreFX}
  MaterialFunctionCall(Material=NamedRerouteUsage(), M=@n1.M, Pre Skinned Local Position=@n1.M, Distance Blend=@n1.M, Fresnel=@n1.M) {materialFunction=&MF_CharacterParent_2023_PostFX}
  NamedRerouteDeclaration(None=@n1.Result) {name=MA, nodeColor=<rgba 0 1 0.0156 1>}
  n2 = NamedRerouteUsage()
  n3 = Panner(
   Coordinate=TextureCoordinate()
   Speed=ComponentMask(None=VectorParameter() {group=00_ScrollDirection, parameterName=ScrollSpeed, defaultValue=<rgba 0 -1 0 0>}.RGB) {g=true, r=true}
  ) {bFractionalPart=true, speedY=1}
  n4 = TextureSampleParameter2D(UVs=@n3) {group=00_Texture Inputs, parameterName=Diffuse, sortPriority=0, texture=/Game/Global/Textures/Default/Blanks/T_MidgreysRGB}
  n5 = TextureSampleParameter2D(UVs=@n3) {
   desc="R = Spec , G = Roughness, B= Metal"
   group=00_Texture Inputs
   parameterName=SRM
   samplerType=SAMPLERTYPE_Masks
   sortPriority=1
   texture=&T_Fortnite_Default_S
  }
  n6 = SetMaterialAttributes!(MaterialAttributes=@n2, Base Color=@n4.RGB, Specular=@n5.R, Roughness=@n5.R, Metallic=@n5.R) {attributeSetTypes [
    #1
    #2
    #3
    #4
   ]}
  n7 = TextureSampleParameter2D() {
   group=00_ScrollDirection
   parameterName=FX
   samplerType=SAMPLERTYPE_Masks
   sortPriority=1
   texture=/BRCosmetics/Animation/Props/Gadgets_Scamper/Textures/T_Scamper_Emote_Prop_FX
  }
  n8 = ChannelMaskParameter(None=@n7.RGB) {group=00_ScrollDirection, maskChannel=Green, parameterName=ScrollMask_Channel, sortPriority=2, defaultValue=<rgba 0 1 0 0>}
  NamedRerouteDeclaration(None=BlendMaterialAttributes(A=@n2, B=@n6, Alpha=@n8.R)) {name=MA OUT, nodeColor=<rgba 1 0.9141 0 1>}
 ]
}
